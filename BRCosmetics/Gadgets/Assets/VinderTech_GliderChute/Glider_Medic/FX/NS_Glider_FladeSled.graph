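# Niagara system for the FladeSled glider cosmetic, drawing rotor, wing-trail and speedline effects.
# HeliRotors pins two multi-renderer mesh particles to skeletal bones, while two Sidewings emitters build bone-pair anim-trail ribbons.
# Speedlines spawns stretched sprites whose velocity shifts with the ForwardDot and RightDot user parameters.
# System timelines drive fades and trail culling, with front-end overrides throughout and time-of-day tinting on rotors and trails.
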NiagaraSystem /BRCosmetics/Gadgets/Assets/VinderTech_GliderChute/Glider_Medic/FX/NS_Glider_FladeSled
dependencies {
 usedDataInterfaces [
  &NiagaraDataInterfaceCurve
  &NiagaraDataInterfaceVector2DCurve
  &NiagaraDataInterfaceColorCurve
  &NiagaraDataInterfaceArrayFloat
  &NiagaraDataInterfaceArrayInt32
  &NiagaraDataInterfaceArrayFloat3
 ]
 usedDynamicInputs [
  &RemapRange
  &MakeIntFromFloat
  &NormalizeVector
  &Boolean_AndOperation
 ]
 usedModules [
  &self:SystemSpawnScript.SystemScriptSource.SystemScriptGraph.NiagaraNodeAssignment_0.SetVariables_#1
  &SystemState
  &Timeline
  &FN_Integrator
 ]
}
emitters {
 HeliRotors {
  emitterUpdateScript [
   &EmitterState(
    Inactive Response=::NewEnumerator1
    Life Cycle Mode=::NewEnumerator1
    Loop Behavior=::NewEnumerator1
    Loop Duration Mode=::NewEnumerator1
    UseLoopDelay=false
   )
   &SpawnBurst_Instantaneous(Spawn Count=2, Use Spawn Probability=false, UseLoopCountLimit=false)
  ]
  particleSpawnScript [
   &InitializeParticle(
    Color=&FNTimeOfDayColor(
     Night Color=<rgba 0.0232 0.0123 0.1301 1>
     Morning Color=<rgba 0.2086 0.1301 0.1912 1>
     Day Color=<rgba 0.2086 0.1301 0.1912 1>
     Evening Color=<rgba 0.2086 0.1301 0.1912 1>
    )
    Mesh Scale Mode=::NewEnumerator4
    UsePositionOffset=false
    Write Material Random=true
   )
   &PartitionParticles(ID to Partition:NiagaraInt32=&ReturnExecIndex())
   /SetVariables(Particles.OpacityForward=1) {isSetParameters=true}
   /SetVariables(
    Particles.BlurSIze:Vector3f=&Multiply_VectorByFloat(
     Vector=<1 1 1>
     Float=&SelectFloatFromArray(Direct Array Index=$Particles.PartitionParticles.Partition, Float Selection Array=DataInterface() {propertyValues {FloatData [0.083,0.345]}})
    )
    Particles.MeshSize1:Vector3f=<2.1 2.1 -0.05>
    Particles.MeshSize2:Vector3f=<2.1 2.1 0.05>
    Particles.SmallRotorSize:Vector3f=<0.5 0.5 -0.025>
   ) {isSetParameters=true}
   &InitialMeshOrientation(Mesh Axis To Orient=<0 0 1>, Mesh Orientation Mode=::NewEnumerator2, Use Rotation Vector=false)
   /SetVariables(
    Particles.BlurvisTag=2
    Particles.VisibilityTag=&SelectIntFromArray(Direct Array Index=$Particles.PartitionParticles.Partition, Int Selection Array=DataInterface() {propertyValues {IntData [1,0]}})
   ) {isSetParameters=true}
  ]
  particleUpdateScript [
   &ParticleState(Kill Particles When Lifetime Has Elapsed=false, Let Infinitely Lived Particles Die When Emitter Deactivates=true)
   &SkeletalMeshLocation(
    Bone / Socket Index=&SelectIntFromArray(Direct Array Index=$Particles.PartitionParticles.Partition, Int Selection Array=DataInterface() {propertyValues {IntData [0,1]}})
    Bone Sampling Mode=::NewEnumerator4
    "Interpolate Mesh Sampling (Update)"=true
    MaskSpawnGroup=false
    Sampled Position Offset=&SelectVectorFromArray(
     Direct Array Index=$Particles.PartitionParticles.Partition
     Vector Selection Array=DataInterface() {propertyValues {FloatData [<-7 0 0>, <15 0 0>]}}
    )
    Skeletal Mesh=$System.Skeletal Mesh
   )
   /SetVariables(
    Particles.BlurColor:LinearColor=&MakeLinearColorFromVectorAndFloat(
     "Vector (RGB)"=&MakeVectorFromLinearColorRGB(Color=&FNTimeOfDayColor(Night Color=<rgba 0.5475 0.45 0.9 1>, Morning Color=<rgba 1 1 1 1>, Day Color=<rgba 1 0.749 0.8275 1>, Evening Color=<rgba 1 1 1 1>))
     "Float (Alpha)"=&FN_FrontEndFloatModifier(Gameplay=&Lerp_Float(A=0.5, B=1, Alpha=$System.Timeline003.Time), Front End=0.25)
    )
    Particles.RotorColor:LinearColor=&MakeLinearColorFromVectorAndFloat(
     "Vector (RGB)"=&MakeVectorFromLinearColorRGB(Color=&FNTimeOfDayColor(Night Color=<rgba 0.5475 0.45 0.9 1>, Morning Color=<rgba 1 1 1 1>, Day Color=<rgba 1 0.749 0.8275 1>, Evening Color=<rgba 1 1 1 1>))
     "Float (Alpha)"=&FN_FrontEndFloatModifier(Gameplay=&Lerp_Float(A=0, B=0.01, Alpha=$System.Timeline003.Time), Front End=0.01)
    )
    Particles.TinyRotorColor:LinearColor=&MakeLinearColorFromVectorAndFloat(
     "Vector (RGB)"=&MakeVectorFromLinearColorRGB(Color=&FNTimeOfDayColor(Night Color=<rgba 0.5475 0.45 0.9 1>, Morning Color=<rgba 1 1 1 1>, Day Color=<rgba 1 0.749 0.8275 1>, Evening Color=<rgba 1 1 1 1>))
     "Float (Alpha)"=&FN_FrontEndFloatModifier(Gameplay=&Lerp_Float(A=0.02, B=0.05, Alpha=$System.Timeline003.Time), Front End=0.02)
    )
   ) {isSetParameters=true}
  ]
  rendererClasses [&NiagaraMeshRendererProperties]
  renderers [
   {rendererClass=&NiagaraMeshRendererProperties, rendererIndex=0}
   {rendererClass=&NiagaraMeshRendererProperties, rendererIndex=1}
   {rendererClass=&NiagaraMeshRendererProperties, rendererIndex=2}
   {rendererClass=&NiagaraMeshRendererProperties, rendererIndex=3}
   {rendererClass=&NiagaraMeshRendererProperties, rendererIndex=4}
  ]
 }
 SidewingsLayer1 {
  emitterUpdateScript [
   &EmitterState()
   &FN_SpawnAnimTrail_v2(Number of Anim Trails=2)
  ]
  particleSpawnScript [
   &PartitionParticles(ID to Partition:NiagaraInt32=$Engine.ExecIndex)
   /Game/Effects/Niagara/Modules/AnimTrail/FN_InitializeAnimTrail/FN_InitializeAnimTrailBoneIndecies(First Bone or Socket Index=1)
   &FN_InitializeAnimTrailBoneIndecies(AnimTrail Selection=1, First Bone or Socket Index=3, Second Bone or Socket Index=2) {module=FN_InitializeAnimTrailBoneIndecies001}
   &FN_InitializeAnimTrail(Color=<rgba 0.1746 0.5583 0.7011 1>, Lifetime=0.2, SkeletalMesh=$System.Skeletal Mesh) {module=FN_InitializeAnimTrail001}
   &AddVelocity(Velocity=&RotateVectorByQuaternion(Vector To Rotate=&MakeVector(X=0, Y=-650, Z=0), Quaternion=<0 0 1 1>))
  ]
  particleUpdateScript [
   &InheritSourceMovement(
    Applied Position Scale=&Multiply_VectorByFloat(
     Vector=<1 1 1>
     Float=&FloatFromCurve(FloatCurve=DataInterface() {propertyValues {Curve {keys [
          {interpMode=RCIM_Cubic, tangentMode=RCTM_User, time=0.66, value=1}
          {arriveTangent=-1.9042, interpMode=RCIM_Cubic, leaveTangent=-1.9042, tangentMode=RCTM_User, time=1, value=0.5}
         ]}}})
    )
    ApplyForce=false
    ApplyPosition=true
    ApplyPositionFalloff=false
   )
   &InterpolateOverTime(Initial Float Value:NiagaraFloat=$System.Timeline.Time, Rate Of Change=0.1, Target Float Value=1, Use Set Interpolation Rate=true)
   &KillParticles(Kill Particles=&SetBoolByFloatComparison(A=$System.Timeline003.Time, B=$Particles.InterpolateOverTime.InterpolateOverTime_Float.Moving Average, Comparison Type=::NewEnumerator6))
   &SolveForcesAndVelocity(Clamp Velocity=false, Limit Acceleration=false)
   &FN_UpdateAnimTrail(SkeletalMesh=$System.Skeletal Mesh) {module=FN_UpdateAnimTrail001}
   &ScaleColor(
    Scale Alpha=&Lerp_Float(
     A=&FloatFromCurve(
      FloatCurve=DataInterface() {propertyValues {Curve {keys [<at 0 1 RCIM_Cubic>,<at 1 0>]}}}
      Scale Curve=&FN_FrontEndFloatModifier(Gameplay=$System.Timeline.Time, Front End=0.6)
     )
     B=&FloatFromCurve(
      FloatCurve=DataInterface() {propertyValues {Curve {keys [<at 0 1.2 RCIM_Cubic>,<at 1 0>]}}}
      Scale Curve=&FN_FrontEndFloatModifier(Gameplay=$System.Timeline.Time, Front End=0.6)
     )
     Alpha=$User.Variant
    )
    Scale RGB=&VectorFromFloat(Value=&FNTimeOfDayScalar(Night Scale=0.4, Morning Scale=0.76, Evening Scale=0.65))
    ScaleA=true
    ScaleRGB=true
   )
   /SetVariables(Particles.Position=&AddVectorToPosition(
     Position=$Particles.Position
     Vector=&TransformVector(
      Vector=&FN_FrontEndVector3Modifier(
       Gameplay=&MaskVectorBySpawnGroup(Masked Vector=<-8 0 0>, Passthrough Vector=<0 0 0>, Spawn Group Mask=$Particles.PartitionParticles.Partition)
       Front End=&MaskVectorBySpawnGroup(Masked Vector=<-8 0 0>, Passthrough Vector=<0 0 0>, Spawn Group Mask=$Particles.PartitionParticles.Partition)
      )
      Destination Space=::Simulation
     )
    )) {isSetParameters=true, module=SetVariables_#2}
   /SetVariables(Particles.RibbonU0Override=$Particles.NormalizedAge) {isSetParameters=true, module=SetVariables_#3}
  ]
  rendererClasses [&NiagaraRibbonRendererProperties]
  renderers [{rendererClass=&NiagaraRibbonRendererProperties}]
 }
 SidewingsLayer2 {
  emitterUpdateScript [
   &EmitterState()
   &FN_SpawnAnimTrail_v2(Number of Anim Trails=4)
  ]
  particleSpawnScript [
   &PartitionParticles(ID to Partition:NiagaraInt32=$Engine.ExecIndex)
   &FN_InitializeAnimTrailBoneIndecies(First Bone or Socket Index=4, Second Bone or Socket Index=5)
   &FN_InitializeAnimTrailBoneIndecies(AnimTrail Selection=1, First Bone or Socket Index=6, Second Bone or Socket Index=7) {module=FN_InitializeAnimTrailBoneIndecies001}
   &FN_InitializeAnimTrailBoneIndecies(AnimTrail Selection=2, First Bone or Socket Index=8, Second Bone or Socket Index=9) {module=FN_InitializeAnimTrailBoneIndecies002}
   &FN_InitializeAnimTrailBoneIndecies(AnimTrail Selection=3, First Bone or Socket Index=10, Second Bone or Socket Index=11) {module=FN_InitializeAnimTrailBoneIndecies003}
   &FN_InitializeAnimTrail(AnimTrailWidth=2, Color=<rgba 0.6458 0.6458 0.6458 1>, Lifetime=0.15, SkeletalMesh=$System.Skeletal Mesh) {module=FN_InitializeAnimTrail001}
   &AddVelocity(Velocity=&RotateVectorByQuaternion(Vector To Rotate=&MakeVector(X=0, Y=-650, Z=0), Quaternion=<0 0 1 1>))
  ]
  particleUpdateScript [
   &InheritSourceMovement(
    Applied Position Scale=&Multiply_VectorByFloat(
     Vector=<1 1 1>
     Float=&FloatFromCurve(FloatCurve=DataInterface() {propertyValues {Curve {keys [
          {interpMode=RCIM_Cubic, tangentMode=RCTM_User, time=0.66, value=1}
          {arriveTangent=-1.9042, interpMode=RCIM_Cubic, leaveTangent=-1.9042, tangentMode=RCTM_User, time=1, value=0.5}
         ]}}})
    )
    ApplyForce=false
    ApplyPosition=true
    ApplyPositionFalloff=false
   )
   &InterpolateOverTime(Initial Float Value:NiagaraFloat=$System.Timeline.Time, Rate Of Change=0.1, Target Float Value=1, Use Set Interpolation Rate=true)
   &KillParticles(Kill Particles=&SetBoolByFloatComparison(A=$System.Timeline003.Time, B=$Particles.InterpolateOverTime.InterpolateOverTime_Float.Moving Average, Comparison Type=::NewEnumerator6))
   &SolveForcesAndVelocity(Clamp Velocity=false, Limit Acceleration=false)
   &FN_UpdateAnimTrail(SkeletalMesh=$System.Skeletal Mesh) {module=FN_UpdateAnimTrail001}
   &ScaleColor(
    Scale Alpha=&FloatFromCurve(
     FloatCurve=DataInterface() {propertyValues {Curve {keys [<at 0 1 RCIM_Cubic>,<at 1 0>]}}}
     Scale Curve=&FN_FrontEndFloatModifier(Gameplay=$System.Timeline.Time, Front End=0.6)
    )
    Scale RGB=&VectorFromFloat(Value=&FNTimeOfDayScalar(Night Scale=0.6, Morning Scale=0.76, Evening Scale=0.65))
    ScaleA=true
    ScaleRGB=true
   )
   &ScaleColor(
    Scale Mode=::NewEnumerator1
    Scale RGBA=&MakeVector4_FromLinearColor(LinearColor=&FNTimeOfDayColor(Night Color=<rgba 0.45 0.556 1 1>, Morning Color=<rgba 1 1 1 1>, Day Color=<rgba 1 1 1 1>, Evening Color=<rgba 1 1 1 1>))
    ScaleRGBA=true
   ) {module=ScaleColor001}
   /SetVariables(Particles.Position=&AddVectorToPosition(
     Position=$Particles.Position
     Vector=&TransformVector(
      Vector=&FN_FrontEndVector3Modifier(
       Gameplay=&MaskVectorBySpawnGroup(Masked Vector=<0 -1 0>, Passthrough Vector=<0 1 0>, Spawn Group Mask=$Particles.PartitionParticles.Partition)
       Front End=&MaskVectorBySpawnGroup(Masked Vector=<0 -1 0>, Passthrough Vector=<0 1 0>, Spawn Group Mask=$Particles.PartitionParticles.Partition)
      )
      Destination Space=::Simulation
     )
    )) {isSetParameters=true, module=SetVariables_#2}
   /SetVariables(Particles.RibbonU0Override=$Particles.NormalizedAge) {isSetParameters=true, module=SetVariables_#3}
  ]
  rendererClasses [&NiagaraRibbonRendererProperties]
  renderers [{rendererClass=&NiagaraRibbonRendererProperties}]
 }
 Speedlines {
  emitterUpdateScript [
   &EmitterState()
   &SpawnRate(
    Spawn Probability=&MakeCustomFloatFromBool(Bool=&SetBoolByFloatComparison(A=$System.Timeline003.Time, B=0), True Float=1, False Float=0)
    SpawnRate=&MakeCustomFloatFromBool(Bool=$User.bIsFrontEndPreview, True Float=50, False Float=90)
    Use Spawn Probability=true
   ) {module=SpawnRate001}
  ]
  particleSpawnScript [
   &InitializeParticle(
    Color=&MakeLinearColorFromVectorAndFloat(
     "Vector (RGB)"=<1 1 1>
     "Float (Alpha)"=&Multiply_Float(
      A=&UniformRangedFloat(Minimum=1, Maximum=1)
      B=&MakeCustomFloatFromBool(
       Bool=$User.bIsFrontEndPreview
       True Float=&FloatFromCurve(
        FloatCurve=DataInterface() {propertyValues {Curve {keys [{tangentMode=RCTM_User, time=0, value=0}, {tangentMode=RCTM_User, time=2, value=1}]}}}
        CurveIndex=$Emitter.Age
        Scale Curve=0.4
       )
       False Float=1
      )
     )
    )
    Lifetime Max=0.2
    Lifetime Min=0.15
    Lifetime Mode=::NewEnumerator1
    Position Offset=&FN_FrontEndVector3Modifier(Gameplay=<0 -50 0>, Front End=<0 50 0>)
    Sprite Size Max=<3 10>
    Sprite Size Min=<2 2>
    Sprite Size Mode=::NewEnumerator2
    UsePositionOffset=true
    Write Material Random=true
   )
   &AddVelocity(Module.Use Falloff Distance=false, Origin Offset=<300 0 0>, Override Seed=false, Velocity Mode=::NewEnumerator1, Velocity Speed=-2500) {module=AddVelocity001}
   &ShapeLocation(
    Handle Radius=0
    Large Radius=&FN_FrontEndFloatModifier(
     Gameplay=&RandomRangeFloat(Minimum=250, Maximum=400)
     Front End=&RandomRangeFloat(Minimum=300, Maximum=350)
    )
    MaskSpawnGroup=false
    Offset=&FN_FrontEndVector3Modifier(Gameplay=<-500 0 0>, Front End=<-350 0 0>)
    Offset Mode=::NewEnumerator0
    Override Seed=false
    Rotation Angle Type:ENiagara_AngleInput=::NewEnumerator0
    Rotation Mode=::NewEnumerator1
    Shape Primitive=::NewEnumerator3
    UseOwnerScale=true
   )
   &AddVelocity(
    Rotation Coordinate Space=::Simulation
    Velocity=&Add_Vector(
     A=&FN_FrontEndVector3Modifier(Gameplay=<0 0 0>, Front End=<-500 0 250>)
     B=&Add_Vector(
      A=&Multiply_VectorByFloat(Vector=<0 550 0>, Float=$User.RightDot)
      B=&Multiply_VectorByFloat(Vector=<222 0 0>, Float=$User.ForwardDot)
     )
    )
    Velocity Speed Scale=&FN_FrontEndFloatModifier(Gameplay=1, Front End=-1)
   )
  ]
  particleUpdateScript [
   &SolveForcesAndVelocity(Clamp Velocity=false, Limit Acceleration=false)
   &ParticleState()
   &ScaleSpriteSize(Scale Factor=&Vector2DFromCurve(
     Vector2Curve=DataInterface() {propertyValues {XCurve {keys [<at 0 0.5 RCIM_Cubic>,<at 0.5 1 RCIM_Cubic>]}, YCurve {keys [<at 0 0 RCIM_Cubic>,<at 0.5 1 RCIM_Cubic>]}}}
     ScaleCurve=&Vector2DFromFloat(Value=1)
    ))
   &ScaleSpriteSizeBySpeed(Max Scale Factor=<1 50>, Min Scale Factor=<0.01 0.05>, Velocity Threshold=1)
   &ScaleColor(
    Scale Alpha=&FloatFromCurve(
     FloatCurve=DataInterface() {propertyValues {Curve {keys [
         <at 0 0 RCIM_Cubic>
         <at 0.3759 1.0207>
         {arriveTangent=-2.221, interpMode=RCIM_Cubic, leaveTangent=-2.221, tangentMode=RCTM_User, time=1, value=0}
        ]}}}
     Scale Curve=&MakeCustomFloatFromBool(
      Bool=$User.bIsFrontEndPreview
      True Float=&MakeCustomFloatFromBool(Bool=$User.bIsFullyDeployed, True Float=0.5, False Float=0)
      False Float=&Lerp_Float(A=0, B=0.5, Alpha=&FN_FrontEndFloatModifier(Gameplay=$System.Timeline.Time, Front End=0.5))
     )
    )
    Scale RGB=&VectorFromFloat(Value=&FN_FrontEndFloatModifier(Gameplay=0.5, Front End=0.33))
    ScaleA=true
    ScaleRGB=true
   )
  ]
  rendererClasses [&NiagaraSpriteRendererProperties]
  renderers [{rendererClass=&NiagaraSpriteRendererProperties}]
 }
}
summary {
 systemName=NS_Glider_FladeSled
 emitters [
  {emitterName=Speedlines, rendererClasses [&NiagaraSpriteRendererProperties]}
  {emitterName=SidewingsLayer1, rendererClasses [&NiagaraRibbonRendererProperties]}
  {emitterName=SidewingsLayer2, rendererClasses [&NiagaraRibbonRendererProperties]}
  {emitterName=HeliRotors, rendererClasses [&NiagaraMeshRendererProperties]}
 ]
 userVariables [
  User.bIsFrontEnd {type=NiagaraBool}
  User.bIsFrontEndPreview {type=NiagaraBool, defaultValue=true}
  User.bIsFullyDeployed {type=NiagaraBool, defaultValue=true}
  User.ForwardDot {type=NiagaraFloat, defaultValue=0, description=""}
  User.Opacity {type=NiagaraFloat, defaultValue=0, description=""}
  User.RibbonIDOffset {type=NiagaraInt32}
  User.RightDot {type=NiagaraFloat, defaultValue=0, description=""}
  User.SkeletalMesh {
   type=NiagaraDataInterfaceSkeletalMesh
   defaultValue {
    struct=NiagaraExt_VariableValue_DataInterface
    value {dataInterface=&self:NiagaraDataInterfaceSkeletalMesh_1}
   }
  }
  User.Variant {type=NiagaraFloat, defaultValue=0, description=""}
 ]
}
